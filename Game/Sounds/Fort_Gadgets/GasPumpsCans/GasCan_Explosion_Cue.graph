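# Sound cue for a gas can explosion, layered by listener distance.
# A mixer sums three branches with close, medium and distant attenuation, each randomly picking a wave and varying pitch.
# Close uses gas can waves, overridden attenuation and a 3P fire sound class; medium and distant reuse propane tank waves under a distant fire class.
# The distant layer is also turned down, and a shared stop-oldest tank explosion concurrency rule limits overlapping instances.
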SoundCue /Game/Sounds/Fort_Gadgets/GasPumpsCans/GasCan_Explosion_Cue
root=Mixer_4
nodes [
 {
  id=Attenuation_0
  children [Modulator_2]
  properties {
   attenuationSettings=/Game/Sounds/Attenuation/Fort_Explosion_Distant_Athena_Att
   attenuationOverrides {falloffDistance=3000, attenuationShapeExtents {x=1000}}
  }
 }
 {
  id=Attenuation_4
  children [Modulator_0]
  properties {
   attenuationSettings=/Game/Sounds/Attenuation/Fort_Explosion_Close_Athena_Att
   bOverrideAttenuation=true
   attenuationOverrides {falloffDistance=5106.2798, stereoSpread=5084.1587, attenuationShapeExtents {x=2554.7803}}
  }
 }
 {
  id=Attenuation_5
  children [Modulator_5]
  properties {
   attenuationSettings=/Game/Sounds/Attenuation/Fort_Explosion_Medium_Athena_Att
   attenuationOverrides {falloffDistance=3000, attenuationShapeExtents {x=1000}}
  }
 }
 {id=Mixer_4, children [SoundClass_3,SoundClass_2,SoundClass_1], properties {inputVolume [1,0.75,0.8]}}
 {id=Modulator_0, children [Random_2], properties {pitchMax=1.1, pitchMin=0.9}}
 {id=Modulator_1, children [Attenuation_4], properties {pitchMax=1.2, pitchMin=0.9}}
 {id=Modulator_2, children [Modulator_3], properties {pitchMax=1.1, pitchMin=1}}
 {id=Modulator_3, children [Random_4], properties {volumeMax=0.4, volumeMin=0.35}}
 {id=Modulator_5, children [Random_1], properties {pitchMax=1.2, pitchMin=1}}
 {id=Random_1, children [WavePlayer_9, WavePlayer_10, WavePlayer_11], properties {weights [1,1,1]}}
 {id=Random_2, children [WavePlayer_0,WavePlayer_3,WavePlayer_4,WavePlayer_6], properties {weights [1,1,1,1]}}
 {id=Random_4, children [WavePlayer_12, WavePlayer_13, WavePlayer_14], properties {weights [1,1,1]}}
 {
  id=SoundClass_1
  children [Attenuation_0]
  properties {soundClassOverride=&Fort_WeaponRanged_Fire_Distant}
 }
 {
  id=SoundClass_2
  children [Attenuation_5]
  properties {soundClassOverride=&Fort_WeaponRanged_Fire_Distant}
 }
 {id=SoundClass_3, children [Modulator_1], properties {soundClassOverride=&Fort_WeaponRanged_Fire_3P_PSM}}
 {id=WavePlayer_0, properties {soundWaveAssetPtr=&/GasCan_Explosion_01}}
 {id=WavePlayer_10, properties {soundWaveAssetPtr=/Game/Sounds/Fort_Gadgets/PropaneTank/PropaneTank_ExplodeMedium_02}}
 {id=WavePlayer_11, properties {soundWaveAssetPtr=/Game/Sounds/Fort_Gadgets/PropaneTank/PropaneTank_ExplodeMedium_03}}
 {id=WavePlayer_12, properties {soundWaveAssetPtr=/Game/Sounds/Fort_Gadgets/PropaneTank/PropaneTank_ExplodeDistant_01}}
 {id=WavePlayer_13, properties {soundWaveAssetPtr=/Game/Sounds/Fort_Gadgets/PropaneTank/PropaneTank_ExplodeDistant_02}}
 {id=WavePlayer_14, properties {soundWaveAssetPtr=/Game/Sounds/Fort_Gadgets/PropaneTank/PropaneTank_ExplodeDistant_03}}
 {id=WavePlayer_3, properties {soundWaveAssetPtr=&/GasCan_Explosion_02}}
 {id=WavePlayer_4, properties {soundWaveAssetPtr=&/GasCan_Explosion_03}}
 {id=WavePlayer_6, properties {soundWaveAssetPtr=&/GasCan_Explosion_04}}
 {id=WavePlayer_9, properties {soundWaveAssetPtr=/Game/Sounds/Fort_Gadgets/PropaneTank/PropaneTank_ExplodeMedium_01}}
]
properties {
 maxDistance=27500
 volumeMultiplier=0.8
 concurrencyOverrides {maxCount=2, volumeScale=0.8762}
 concurrencySet [/Game/Sounds/Concurrency_Rules/TankExplosion_StopOldest_Conc]
}
