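# Sound cue for pouring from a gas can, combining a one-shot pour start with a sustained loop.
# A mixer layers a randomly chosen start variant, pitched slightly up, with a looping wave that enters after a short fixed delay.
# It uses the gas pump and can attenuation and the gadget items sound class, with a per-owner concurrency override that stops the oldest instance.
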SoundCue /Game/Sounds/Fort_Gadgets/GasPumpsCans/GasCan_Pour_Loop_Cue
root=Mixer_0
nodes [
 {id=Delay_0, children [WavePlayer_4], properties {delayMax=0.2, delayMin=0.2}}
 {id=Mixer_0, children [Modulator_0,Modulator_1], properties {inputVolume [0.7,0.8]}}
 {id=Modulator_0, children [Random_0], properties {pitchMax=1.15, pitchMin=1.05}}
 {id=Modulator_1, children [Delay_0]}
 {id=Random_0, children [WavePlayer_0,WavePlayer_1,WavePlayer_2,WavePlayer_3], properties {weights [1,1,1,1]}}
 {id=WavePlayer_0, properties {soundWaveAssetPtr=&/GasCan_Pour_Start_01}}
 {id=WavePlayer_1, properties {soundWaveAssetPtr=&/GasCan_Pour_Start_02}}
 {id=WavePlayer_2, properties {soundWaveAssetPtr=&/GasCan_Pour_Start_03}}
 {id=WavePlayer_3, properties {soundWaveAssetPtr=&/GasCan_Pour_Start_04}}
 {id=WavePlayer_4, properties {bLooping=true, soundWaveAssetPtr=&/GasCan_Pour_Loop_01}}
]
properties {
 attenuationSettings=/Game/Sounds/Attenuation/Fort_GasPumpsCans_Att
 bOverrideConcurrency=true
 maxDistance=2500
 soundClassObject=&Fort_Gadgets_Items
 volumeMultiplier=0.22
 concurrencyOverrides {bLimitToOwner=true, maxCount=2, resolutionRule=StopOldest, voiceStealReleaseTime=1, volumeScaleAttackTime=0.5}
}
