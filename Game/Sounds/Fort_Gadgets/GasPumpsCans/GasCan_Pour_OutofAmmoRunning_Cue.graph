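# Sound cue for a gas can pour gadget running out of ammo, under the gadget item sound class.
# A modulator feeds an equal-weight random pick among four out-of-ammo pour waves.
# The cue plays quietly at lowered pitch with gas pump and can attenuation.
# Its own concurrency override caps simultaneous voices and adds a retrigger time and voice-steal release.
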SoundCue /Game/Sounds/Fort_Gadgets/GasPumpsCans/GasCan_Pour_OutofAmmoRunning_Cue
root=Modulator_0
nodes [
 {id=Modulator_0, children [Random_2]}
 {id=Random_2, children [WavePlayer_10, WavePlayer_12, WavePlayer_14, WavePlayer_15], properties {weights [1,1,1,1]}}
 {id=WavePlayer_10, properties {soundWaveAssetPtr=&/GasCan_Pour_OutofAmmo_02}}
 {id=WavePlayer_12, properties {soundWaveAssetPtr=&/GasCan_Pour_OutofAmmo_03}}
 {id=WavePlayer_14, properties {soundWaveAssetPtr=&/GasCan_Pour_OutofAmmo_05}}
 {id=WavePlayer_15, properties {soundWaveAssetPtr=&/GasCan_Pour_OutofAmmo_06}}
]
properties {
 attenuationSettings=/Game/Sounds/Attenuation/Fort_GasPumpsCans_Att
 bOverrideConcurrency=true
 maxDistance=2500
 pitchMultiplier=0.8
 soundClassObject=&Fort_Gadgets_Items
 volumeMultiplier=0.13
 concurrencyOverrides {maxCount=4, retriggerTime=0.6, voiceStealReleaseTime=1}
}
