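# Sound cue for the gas can out-of-ammo pour sound, classed under gadget items.
# A Random node picks one of five equally weighted wave variations and feeds a Modulator at the root.
# The cue plays quietly with slightly raised pitch, uses the shared gas pumps and cans attenuation, and overrides concurrency to limit simultaneous instances.
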SoundCue /Game/Sounds/Fort_Gadgets/GasPumpsCans/GasCan_Pour_OutofAmmo_Cue
root=Modulator_0
nodes [
 {id=Modulator_0, children [Random_0]}
 {id=Random_0, children [WavePlayer_2, WavePlayer_4, WavePlayer_9, WavePlayer_10, WavePlayer_11], properties {weights [1,1,1,1,1]}}
 {id=WavePlayer_10, properties {soundWaveAssetPtr=&/GasCan_Pour_OutofAmmo_05}}
 {id=WavePlayer_11, properties {soundWaveAssetPtr=&/GasCan_Pour_OutofAmmo_06}}
 {id=WavePlayer_2, properties {soundWaveAssetPtr=&/GasCan_Pour_OutofAmmo_02}}
 {id=WavePlayer_4, properties {soundWaveAssetPtr=&/GasCan_Pour_OutofAmmo_03}}
 {id=WavePlayer_9, properties {soundWaveAssetPtr=&/GasCan_Pour_OutofAmmo_04}}
]
properties {
 attenuationSettings=/Game/Sounds/Attenuation/Fort_GasPumpsCans_Att
 bOverrideConcurrency=true
 maxDistance=2500
 pitchMultiplier=1.1
 soundClassObject=&Fort_Gadgets_Items
 volumeMultiplier=0.13
 concurrencyOverrides {maxCount=4}
}
