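# Sound cue for the gas can pour-stop effect, played when pouring from a gas can ends.
# A short fixed delay leads into a slight random upward pitch modulation, which plays one of three stop waves picked at equal weight.
# It plays quietly through the gas pumps and cans attenuation in the gadget items sound class.
# Overridden concurrency caps it at two per owner, stopping the oldest.
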SoundCue /Game/Sounds/Fort_Gadgets/GasPumpsCans/GasCan_Pour_Stop_Cue
root=Delay_0
nodes [
 {id=Delay_0, children [Modulator_0], properties {delayMax=0.1, delayMin=0.1}}
 {id=Modulator_0, children [Random_0], properties {pitchMax=1.1, pitchMin=1}}
 {id=Random_0, children [WavePlayer_0,WavePlayer_1,WavePlayer_2], properties {weights [1,1,1]}}
 {id=WavePlayer_0, properties {soundWaveAssetPtr=&/GasCan_Pour_Stop_01}}
 {id=WavePlayer_1, properties {soundWaveAssetPtr=&/GasCan_Pour_Stop_02}}
 {id=WavePlayer_2, properties {soundWaveAssetPtr=&/GasCan_Pour_Stop_03}}
]
properties {
 attenuationSettings=/Game/Sounds/Attenuation/Fort_GasPumpsCans_Att
 bOverrideConcurrency=true
 maxDistance=2500
 pitchMultiplier=1.05
 soundClassObject=&Fort_Gadgets_Items
 volumeMultiplier=0.15
 concurrencyOverrides {bLimitToOwner=true, maxCount=2, resolutionRule=StopOldest, voiceStealReleaseTime=1, volumeScaleAttackTime=0.5}
}
