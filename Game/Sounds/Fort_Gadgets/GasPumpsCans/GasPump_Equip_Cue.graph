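# Sound cue for equipping the gas pump item, part of the Fort gadgets gas pumps and cans sounds.
# A Random node picks one of three equally weighted equip waves and feeds it through a Modulator for variation.
# It plays quietly and slightly raised in pitch on the gadget items sound class, with shared gas pump and can attenuation and its own concurrency limit.
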SoundCue /Game/Sounds/Fort_Gadgets/GasPumpsCans/GasPump_Equip_Cue
root=Modulator_0
nodes [
 {id=Modulator_0, children [Random_0]}
 {id=Random_0, children [WavePlayer_0,WavePlayer_1,WavePlayer_2], properties {weights [1,1,1]}}
 {id=WavePlayer_0, properties {soundWaveAssetPtr=&/GasPump_Equip_01}}
 {id=WavePlayer_1, properties {soundWaveAssetPtr=&/GasPump_Equip_02}}
 {id=WavePlayer_2, properties {soundWaveAssetPtr=&/GasPump_Equip_03}}
]
properties {
 attenuationSettings=/Game/Sounds/Attenuation/Fort_GasPumpsCans_Att
 bOverrideConcurrency=true
 maxDistance=2500
 pitchMultiplier=1.1
 soundClassObject=&Fort_Gadgets_Items
 volumeMultiplier=0.3
 concurrencyOverrides {maxCount=4}
}
