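# Sound cue for pouring at a gas pump, made of a start one-shot layered with a sustained pour loop.
# A mixer plays a modulated random pick of three start sounds alongside a modulated looping wave held back by a short fixed delay, so the loop follows the onset.
# The loop layer is mixed louder, and the cue uses the gas pump and can attenuation with the gadget items sound class.
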SoundCue /Game/Sounds/Fort_Gadgets/GasPumpsCans/GasPump_Pour_Loop_Cue
root=Mixer_0
nodes [
 {id=Delay_0, children [WavePlayer_0], properties {delayMax=0.2, delayMin=0.2}}
 {id=Mixer_0, children [Modulator_0,Modulator_1], properties {inputVolume [0.72,1.5]}}
 {id=Modulator_0, children [Random_2]}
 {id=Modulator_1, children [Delay_0]}
 {id=Random_2, children [WavePlayer_10, WavePlayer_11, WavePlayer_12], properties {weights [1,1,1]}}
 {id=WavePlayer_0, properties {bLooping=true, soundWaveAssetPtr=&/GasPump_Pour_Loop_02}}
 {id=WavePlayer_10, properties {soundWaveAssetPtr=&/GasPump_Pour_Start_01}}
 {id=WavePlayer_11, properties {soundWaveAssetPtr=&/GasPump_Pour_Start_02}}
 {id=WavePlayer_12, properties {soundWaveAssetPtr=&/GasPump_Pour_Start_03}}
]
properties {
 attenuationSettings=/Game/Sounds/Attenuation/Fort_GasPumpsCans_Att
 maxDistance=2500
 soundClassObject=&Fort_Gadgets_Items
 volumeMultiplier=0.35
}
