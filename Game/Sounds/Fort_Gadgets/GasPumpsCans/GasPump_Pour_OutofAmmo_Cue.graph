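# Sound cue for a gas pump pour running out of ammo, among the gas pump and can gadget sounds.
# A Random node picks one of four equally weighted out-of-ammo wave variations per play.
# It plays at reduced volume in the gadget items sound class, using the shared gas pump/can attenuation and its own concurrency override capping simultaneous instances.
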SoundCue /Game/Sounds/Fort_Gadgets/GasPumpsCans/GasPump_Pour_OutofAmmo_Cue
root=Random_1
nodes [
 {id=Random_1, children [WavePlayer_5,WavePlayer_6,WavePlayer_7,WavePlayer_8], properties {weights [1,1,1,1]}}
 {id=WavePlayer_5, properties {soundWaveAssetPtr=&/GasPump_Pour_OutofAmmo_01}}
 {id=WavePlayer_6, properties {soundWaveAssetPtr=&/GasPump_Pour_OutofAmmo_02}}
 {id=WavePlayer_7, properties {soundWaveAssetPtr=&/GasPump_Pour_OutofAmmo_03}}
 {id=WavePlayer_8, properties {soundWaveAssetPtr=&/GasPump_Pour_OutofAmmo_04}}
]
properties {
 attenuationSettings=/Game/Sounds/Attenuation/Fort_GasPumpsCans_Att
 bOverrideConcurrency=true
 maxDistance=2500
 soundClassObject=&Fort_Gadgets_Items
 volumeMultiplier=0.3
 concurrencyOverrides {maxCount=4}
}
