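# Sound cue for the gas pump pre-explosion effect in the Fort gadgets sound set.
# It plays a borrowed Husk propane steam-rise wave through a modulator that randomizes pitch, allowing it to rise above normal.
# The cue uses the gas pumps and cans attenuation and Fort_Gadgets sound class, with overridden concurrency that allows only one instance at a time.
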SoundCue /Game/Sounds/Fort_Gadgets/GasPumpsCans/GasPump_PreExplode_Cue
root=Modulator_0
nodes [
 {id=Modulator_0, children [WavePlayer_0], properties {pitchMax=1.3}}
 {id=WavePlayer_0, properties {soundWaveAssetPtr=/Game/Sounds/Fort_Monster_Husk/SFX/propane_steam_rise}}
]
properties {
 attenuationSettings=/Game/Sounds/Attenuation/Fort_GasPumpsCans_Att
 bOverrideConcurrency=true
 maxDistance=2500
 soundClassObject=&Fort_Gadgets
 volumeMultiplier=0.6
 concurrencyOverrides {maxCount=1, volumeScaleAttackTime=0.1}
}
